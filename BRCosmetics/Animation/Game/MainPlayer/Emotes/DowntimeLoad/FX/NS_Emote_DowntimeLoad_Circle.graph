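# Niagara sprite effect for the DowntimeLoad emote that draws a glowing, spinning loading circle.
# Three burst-spawned sprite emitters (circle, orange glitch-lines overlay, soft glow behind) scale in from zero and persist past their lifetime.
# Circle and glitch lines derive rotation from engine time, indexing an array of 45-degree angles like a spinner.
# The glow pulses brightness over a looping particle lifetime, with its scale-in keyed to system age.
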NiagaraSystem /BRCosmetics/Animation/Game/MainPlayer/Emotes/DowntimeLoad/FX/NS_Emote_DowntimeLoad_Circle
dependencies {
 usedDataInterfaces [&NiagaraDataInterfaceCurve, &NiagaraDataInterfaceArrayFloat]
 usedDynamicInputs []
 usedModules [&SystemState]
}
emitters {
 Circle_Rotate001 {
  emitterUpdateScript [
   &EmitterState(Life Cycle Mode=::NewEnumerator1, Loop Behavior=::NewEnumerator1, Loop Duration Mode=::NewEnumerator1, UseLoopDelay=false)
   &SpawnBurst_Instantaneous(Use Spawn Probability=false, UseLoopCountLimit=false)
  ]
  particleSpawnScript [&InitializeParticle(
    Color=<rgba 1 1 1 0.75>
    Lifetime=5
    Sprite Rotation Angle=-60
    Sprite Rotation Mode=::NewEnumerator0
    Sprite Size Mode=::NewEnumerator3
    Uniform Sprite Size=20
    UsePositionOffset=false
    Write Material Random=true
   )]
  particleUpdateScript [
   &ParticleState(Kill Particles When Lifetime Has Elapsed=false)
   &ScaleSpriteSize(Uniform Curve Sprite Scale=DataInterface() {propertyValues {Curve {
       defaultValue=0
       keys [{interpMode=RCIM_Cubic, tangentMode=RCTM_User, time=0, value=0}, {interpMode=RCIM_Cubic, tangentMode=RCTM_User, time=0.09, value=1}]
      }}})
   /SetVariables(Particles.SpriteRotation:NiagaraFloat=&SelectFloatFromArray(
     Direct Array Index=&MakeIntFromFloat(Float=&Multiply_Float(A=$Engine.Time, B=5))
     Interpolation=&Frac_Float(Float=$Engine.Time)
     Direct Array Mode=::NewEnumerator1
     Normalized Interpolation=false
     Float Selection Array=DataInterface() {propertyValues {FloatData [0,45,90,135,180,225,270,315]}}
    )) {isSetParameters=true, module=SetVariables_#1}
  ]
  rendererClasses [&NiagaraSpriteRendererProperties]
  renderers [{rendererClass=&NiagaraSpriteRendererProperties}]
 }
 Glitch_Lines {
  emitterSpawnScript [/SetVariables(Emitter.U Scale Power:NiagaraFloat=&Waveform("Waveform [1]"=::NewEnumerator0, "[1] Frequency"=20, "[1] Amplitude Min/Max"=<0.5 0.6>)) {isSetParameters=true}]
  emitterUpdateScript [
   &EmitterState(Life Cycle Mode=::NewEnumerator1, Loop Behavior=::NewEnumerator1, Loop Duration Mode=::NewEnumerator1, UseLoopDelay=false) {module=EmitterState001}
   &SpawnBurst_Instantaneous(Spawn Time=0.05, Use Spawn Probability=false, UseLoopCountLimit=false)
  ]
  particleSpawnScript [
   &InitializeParticle(
    Color=<rgba 1 0.5765 0.2269 0.3>
    Lifetime=5
    Sprite Rotation Angle=-60
    Sprite Rotation Mode=::NewEnumerator0
    Sprite Size Mode=::NewEnumerator3
    Uniform Sprite Size=23
    UsePositionOffset=false
    Write Material Random=true
   ) {module=InitializeParticle002}
   &CameraOffset(Camera Offset Amount=1)
  ]
  particleUpdateScript [
   &ParticleState(Kill Particles When Lifetime Has Elapsed=false)
   &ScaleSpriteSize(Uniform Curve Sprite Scale=DataInterface() {propertyValues {Curve {
       defaultValue=0
       keys [
        {arriveTangent=16.8079, interpMode=RCIM_Cubic, leaveTangent=16.8079, tangentMode=RCTM_User, time=0, value=0}
        {interpMode=RCIM_Cubic, tangentMode=RCTM_User, time=0.08, value=1}
       ]
      }}}) {module=ScaleSpriteSize001}
   /SetVariables(Particles.SpriteRotation:NiagaraFloat=&SelectFloatFromArray(
     Direct Array Index=&MakeIntFromFloat(Float=&Multiply_Float(A=$Engine.Time, B=5))
     Interpolation=&Frac_Float(Float=$Engine.Time)
     Direct Array Mode=::NewEnumerator1
     Normalized Interpolation=false
     Float Selection Array=DataInterface() {propertyValues {FloatData [0,45,90,135,180,225,270,315]}}
    )) {isSetParameters=true, module=SetVariables_#1}
  ]
  rendererClasses [&NiagaraSpriteRendererProperties]
  renderers [{rendererClass=&NiagaraSpriteRendererProperties}]
 }
 Glow001 {
  emitterUpdateScript [
   &EmitterState(Life Cycle Mode=::NewEnumerator1, Loop Behavior=::NewEnumerator1, Loop Duration Mode=::NewEnumerator1, UseLoopDelay=false)
   &SpawnBurst_Instantaneous(Use Spawn Probability=false, UseLoopCountLimit=false)
  ]
  particleSpawnScript [
   &InitializeParticle(
    Color=<rgba 1 0.7322 0.5185 0.12>
    Lifetime=2
    Sprite Size Mode=::NewEnumerator3
    Uniform Sprite Size=37
    UsePositionOffset=false
    Write Material Random=true
   )
   &CameraOffset(Camera Offset Amount=-5)
  ]
  particleUpdateScript [
   &ScaleColor(
    Scale RGB=&Multiply_VectorByFloat(
     Vector=<1 1 1>
     Float=&FloatFromCurve(FloatCurve=DataInterface() {propertyValues {Curve {keys [<at 0 0.75 RCIM_Cubic>,<at 0.25 1.1 RCIM_Cubic>,<at 0.5 0.75 RCIM_Cubic>,<at 0.75 1.1 RCIM_Cubic>,<at 1 0.75 RCIM_Cubic>]}}})
    )
    ScaleA=true
    ScaleRGB=true
   )
   &ParticleState(Kill Particles When Lifetime Has Elapsed=false, Loop Particles Lifetime=true)
   &ScaleSpriteSize(
    Uniform Curve Index=$System.Age
    Uniform Curve Sprite Scale=DataInterface() {propertyValues {Curve {
       defaultValue=0
       keys [{interpMode=RCIM_Cubic, tangentMode=RCTM_User, time=0, value=0}, {interpMode=RCIM_Cubic, tangentMode=RCTM_User, time=0.4, value=1}]
      }}}
   )
  ]
  rendererClasses [&NiagaraSpriteRendererProperties]
  renderers [{rendererClass=&NiagaraSpriteRendererProperties}]
 }
}
summary {
 systemName=NS_Emote_DowntimeLoad_Circle
 emitters [
  {emitterName=Circle_Rotate001, rendererClasses [&NiagaraSpriteRendererProperties]}
  {emitterName=Glitch_Lines, rendererClasses [&NiagaraSpriteRendererProperties]}
  {emitterName=Glow001, rendererClasses [&NiagaraSpriteRendererProperties]}
 ]
 userVariables [
  User.bIsFrontEnd {type=NiagaraBool}
  User.bIsFrontEndPreview {type=NiagaraBool}
 ]
}
